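# Physics asset for the Dragon Mask male medium body cosmetic, covering a four-bone strand on the character's right side near the thigh.
# Capsule bodies cover the thigh twist bone, a small driver bone and the four strand bones.
# Constraints chain thigh to driver to strands, with free rotation at the driver.
# Strand joints have angular limits, tightest at the first and wider down the chain.
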PhysicsAsset /BRCosmetics/Characters/Player/Male/Medium/Bodies/M_MED_DragonMask/Meshes/M_MED_Dragon_Mask_Physics
bodies {
 driver_strand_01_r [{shapeType=Capsule, length=0.505, radius=0.505}]
 strand_01_r [{shapeType=Capsule, length=5.1003, radius=2.4923, center=<-0.5491 0.0375 -1.6452>, rotation=<rot 0 -57.3987 -70.1581>}]
 strand_02_r [{shapeType=Capsule, length=2, radius=1.3937, center=<1.4009 -0.1929 0.0361>, rotation=<rot 0 90.721 -89.0686>}]
 strand_03_r [{shapeType=Capsule, length=3.1235, radius=1.5189, center=<2.6119 -0.069 0.1824>, rotation=<rot 0 97.5519 -80.984>}]
 strand_04_r [{shapeType=Capsule, length=2.0894, radius=1.704, center=<4.0156 0.0374 0.458>, rotation=<rot 39.4074 -90 26.4104>}]
 thigh_twist_01_r [{shapeType=Capsule, length=22.5664, radius=8.6528, rotation=<rot 90 0 0>}]
}
constraints [
 {
  bone1Name=driver_strand_01_r
  bone2Name=thigh_twist_01_r
  swing1LimitDegrees=45
  swing1Motion=Free
  swing2LimitDegrees=45
  swing2Motion=Free
  twistLimitDegrees=45
  twistMotion=Free
 }
 {bone1Name=strand_01_r, bone2Name=driver_strand_01_r, swing1LimitDegrees=30, swing2LimitDegrees=35, twistLimitDegrees=15}
 {bone1Name=strand_02_r, bone2Name=strand_01_r, swing1LimitDegrees=50, swing2LimitDegrees=60, twistLimitDegrees=20}
 {bone1Name=strand_03_r, bone2Name=strand_02_r, swing1LimitDegrees=50, swing2LimitDegrees=60, twistLimitDegrees=20}
 {bone1Name=strand_04_r, bone2Name=strand_03_r, swing1LimitDegrees=50, swing2LimitDegrees=60, twistLimitDegrees=20}
]
