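# Physics asset for the left-side tassels on the medium male Dragon Mask player body, defining their bodies and joints for physical motion.
# A six-capsule tassel chain hangs from a small box driver, which joins a large left-thigh capsule with all angular motion free.
# Three tube capsules branch off the first three tassel bones.
# Tassel and tube joints set per-joint swing and twist limits, generally wider toward the chain's tip.
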PhysicsAsset /BRCosmetics/Characters/Player/Male/Medium/Bodies/M_MED_DragonMask/Meshes/M_MED_Dragon_Mask_TasselsLPhysics
bodies {
 driver_tassel_01_l [{shapeType=Box, extentX=1.01, extentY=1.01, extentZ=1.01, rotation=<rot 0 -90 0>}]
 tassel_01_l [{shapeType=Capsule, length=13.25, radius=2.75, center=<2.1054 -0.0001 -0.122>, rotation=<rot -7.9345 -0.0673 2.3347>}]
 tassel_02_l [{shapeType=Capsule, length=2.25, radius=1, center=<1.1693 0 0.1184>, rotation=<rot 89.5219 -0.0061 -0.0061>}]
 tassel_03_l [{shapeType=Capsule, length=5, radius=1, center=<1.8026 0 -0.0358>, rotation=<rot 88.8618 0.0007 0.0006>}]
 tassel_04_l [{shapeType=Capsule, length=4, radius=1, center=<1.1425 0 -0.0259>, rotation=<rot 88.7032 -0.0003 -0.0002>}]
 tassel_05_l [{shapeType=Capsule, length=2.25, radius=1.25, center=<0.8788 0 -0.0642>, rotation=<rot 90 -0.0049 -0.0049>}]
 tassel_06_l [{shapeType=Capsule, length=1.75, radius=1.75, center=<0.7266 -0.0001 -0.017>, rotation=<rot 88.6625 0 0.0001>}]
 tassel_tube_01_l [{shapeType=Capsule, length=6, radius=2, center=<-5.1823 -0.3735 0.4475>, rotation=<rot 89.7545 -179.9969 -179.9969>}]
 tassel_tube_02_l [{shapeType=Capsule, length=7.25, radius=2, center=<-4.9601 0 -0.1645>, rotation=<rot 87.9296 -179.9999 -179.9999>}]
 tassel_tube_03_l [{shapeType=Capsule, length=5.75, radius=2.5, center=<-5.6069 -0.1987 0.4358>, rotation=<rot 87.9707 -0.0022 -0.0021>}]
 thigh_l [{shapeType=Capsule, length=31.5101, radius=11.8055, center=<-13.1943 -0.0007 2.6632>, rotation=<rot -88.7537 0 0>}]
}
constraints [
 {bone1Name=tassel_01_l, bone2Name=driver_tassel_01_l, swing1LimitDegrees=40, swing2LimitDegrees=35, twistLimitDegrees=75}
 {bone1Name=tassel_02_l, bone2Name=tassel_01_l, swing1LimitDegrees=45, swing2LimitDegrees=45, twistLimitDegrees=80}
 {bone1Name=tassel_03_l, bone2Name=tassel_02_l, swing1LimitDegrees=80, swing2LimitDegrees=80, twistLimitDegrees=80}
 {bone1Name=tassel_04_l, bone2Name=tassel_03_l, swing1LimitDegrees=80, swing2LimitDegrees=80, twistLimitDegrees=80}
 {bone1Name=tassel_05_l, bone2Name=tassel_04_l, swing1LimitDegrees=80, swing2LimitDegrees=80, twistLimitDegrees=80}
 {bone1Name=tassel_06_l, bone2Name=tassel_05_l, swing1LimitDegrees=80, swing2LimitDegrees=80, twistLimitDegrees=80}
 {
  bone1Name=driver_tassel_01_l
  bone2Name=thigh_l
  swing1LimitDegrees=45
  swing1Motion=Free
  swing2LimitDegrees=45
  swing2Motion=Free
  twistLimitDegrees=45
  twistMotion=Free
 }
 {bone1Name=tassel_tube_01_l, bone2Name=tassel_01_l, swing1LimitDegrees=40, swing2LimitDegrees=55, twistLimitDegrees=20}
 {bone1Name=tassel_tube_03_l, bone2Name=tassel_03_l, swing1LimitDegrees=45, swing2LimitDegrees=45, twistLimitDegrees=28.5714}
 {bone1Name=tassel_tube_02_l, bone2Name=tassel_02_l, swing1LimitDegrees=60, swing2LimitDegrees=60, twistLimitDegrees=60}
]
