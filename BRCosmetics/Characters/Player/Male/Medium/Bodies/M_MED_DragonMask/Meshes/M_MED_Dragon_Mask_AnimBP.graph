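# Post-process style AnimBP for the Dragon Mask male body cosmetic, adding secondary motion to coat, shorts, collar and tassels over a pose copied from the attached parent mesh.
# AnimDynamics drives coat, shorts and back collar.
# Additive bone corrections follow, blended by arm, head and calf alphas that are neither declared nor set here.
# A strand constraint, fixed offsets and two RigidBody simulations, based on pelvis and left thigh, finish the chain.
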AnimBlueprint /BRCosmetics/Characters/Player/Male/Medium/Bodies/M_MED_DragonMask/Meshes/M_MED_Dragon_Mask_AnimBP
target_skeleton=&/M_MED_Dragon_Mask_Skeleton
graphs [
 AnimGraph {nodes [
   n1 = Get CalfRotBack_R() -> (CalfRotBack_R:float)
   n2 = Get ArmUpMore_R() -> (ArmUpMore_R:float)
   n3 = Get ShortsAlpha() -> (ShortsAlpha:float)
   n4 = Anim Dynamics - Bone: dyn_coat_fr(ComponentPose=Local To Component(LocalPose=Copy Pose From Mesh() {bUseAttachedParent=true}), Alpha=0.75, ExternalForce="0,0,0") {
    angularDampingOverride=0.98
    bGravityOverrideInSimSpace=true
    bOverrideAngularDamping=true
    bOverrideLinearDamping=true
    gravityScale=0.35
    lODThreshold=1
    linearDampingOverride=0.98
    numSolverIterationsPostUpdate=4
    numSolverIterationsPreUpdate=16
    simulationSpace=RootRelative
    alphaBoolBlend {blendOption=Linear}
    boundBone {boneName=dyn_coat_fr}
    physicsBodyDefinitions [{
      sphereCollisionRadius=0
      boundBone {boneName=dyn_coat_fr}
      boxExtents=<5 5 5>
      constraintSetup {angularLimitsMax=<5 15 15>, angularLimitsMin=<-45 -15 -15>, angularTarget=<0 0 0>}
      localJointOffset=<0 0 -0.5>
     }]
   }
   n5 = Anim Dynamics - Bone: dyn_coat_br(ComponentPose=@n4, Alpha=0.95, ExternalForce="0,0,0") {
    angularDampingOverride=0.995
    angularSpringConstant=80
    bAngularSpring=true
    bGravityOverrideInSimSpace=true
    bOverrideAngularDamping=true
    bOverrideLinearDamping=true
    gravityScale=0.35
    lODThreshold=1
    linearDampingOverride=0.98
    numSolverIterationsPostUpdate=4
    numSolverIterationsPreUpdate=16
    simulationSpace=RootRelative
    alphaBoolBlend {blendOption=Linear}
    boundBone {boneName=dyn_coat_br}
    physicsBodyDefinitions [{
      sphereCollisionRadius=0
      boundBone {boneName=dyn_coat_br}
      boxExtents=<5 5 5>
      constraintSetup {angularLimitsMax=<5 15 15>, angularLimitsMin=<-45 -15 -15>, angularTarget=<0 0 0>}
      localJointOffset=<0 0 -0.5>
     }]
   }
   n6 = Anim Dynamics - Bone: dyn_coat_r(ComponentPose=@n5, Alpha=0.75, ExternalForce="0,0,0") {
    angularDampingOverride=0.98
    bGravityOverrideInSimSpace=true
    bOverrideAngularDamping=true
    bOverrideLinearDamping=true
    gravityScale=0.31
    lODThreshold=1
    linearDampingOverride=0.98
    numSolverIterationsPostUpdate=4
    numSolverIterationsPreUpdate=16
    simulationSpace=RootRelative
    alphaBoolBlend {blendOption=Linear}
    boundBone {boneName=dyn_coat_r}
    componentAppliedLinearAccClamp=<1000 1000 1000>
    componentLinearAccScale=<0.5 0.5 0.5>
    componentLinearVelScale=<0.5 0.5 0.5>
    physicsBodyDefinitions [{
      sphereCollisionRadius=0
      boundBone {boneName=dyn_coat_r}
      boxExtents=<4 4 4>
      constraintSetup {angularLimitsMax=<5 35 15>, angularLimitsMin=<-45 -15 -15>, angularTarget=<0 0 0>}
      localJointOffset=<0 0 -0.75>
     }]
   }
   n7 = Anim Dynamics - Bone: dyn_coat_front_r(ComponentPose=@n6, Alpha=0.75, ExternalForce="0,0,0") {
    angularDampingOverride=0.98
    bGravityOverrideInSimSpace=true
    bOverrideAngularDamping=true
    bOverrideLinearDamping=true
    gravityScale=0.35
    lODThreshold=1
    linearDampingOverride=0.98
    numSolverIterationsPostUpdate=4
    numSolverIterationsPreUpdate=16
    simulationSpace=RootRelative
    alphaBoolBlend {blendOption=Linear}
    boundBone {boneName=dyn_coat_front_r}
    physicsBodyDefinitions [{
      sphereCollisionRadius=0
      boundBone {boneName=dyn_coat_front_r}
      boxExtents=<5 5 5>
      constraintSetup {angularLimitsMax=<5 15 15>, angularLimitsMin=<-45 -15 -15>, angularTarget=<0 0 0>}
      localJointOffset=<0 0 -0.5>
     }]
   }
   n8 = Anim Dynamics - Bone: dyn_coat_bl(ComponentPose=@n7, Alpha=0.75, ExternalForce="0,0,0") {
    angularDampingOverride=0.98
    bGravityOverrideInSimSpace=true
    bOverrideAngularDamping=true
    bOverrideLinearDamping=true
    gravityScale=0.35
    lODThreshold=1
    linearDampingOverride=0.98
    numSolverIterationsPostUpdate=4
    numSolverIterationsPreUpdate=16
    simulationSpace=RootRelative
    alphaBoolBlend {blendOption=Linear}
    boundBone {boneName=dyn_coat_bl}
    physicsBodyDefinitions [{
      sphereCollisionRadius=0
      boundBone {boneName=dyn_coat_bl}
      boxExtents=<5 5 5>
      constraintSetup {angularLimitsMax=<5 15 15>, angularLimitsMin=<-45 -15 -15>, angularTarget=<0 0 0>}
      localJointOffset=<0 0 0.5>
     }]
   }
   n9 = Anim Dynamics - Bone: dyn_coat_bl(ComponentPose=@n8, Alpha=0.875, ExternalForce="0,0,0") {
    angularDampingOverride=0.9995
    angularSpringConstant=30
    bAngularSpring=true
    bGravityOverrideInSimSpace=true
    bOverrideAngularDamping=true
    bOverrideLinearDamping=true
    gravityScale=0.33
    lODThreshold=1
    linearDampingOverride=0.98
    numSolverIterationsPostUpdate=4
    numSolverIterationsPreUpdate=16
    simulationSpace=RootRelative
    alphaBoolBlend {blendOption=Linear}
    boundBone {boneName=dyn_coat_bl}
    physicsBodyDefinitions [{
      sphereCollisionRadius=0
      boundBone {boneName=dyn_coat_bl}
      boxExtents=<6 6 6>
      constraintSetup {angularLimitsMax=<5 16 15>, angularLimitsMin=<-45 -20 -15>, angularTarget=<0 0 0>}
      localJointOffset=<0 0 0.5>
     }]
   }
   n10 = Anim Dynamics - Bone: dyn_coat_l(ComponentPose=@n9, Alpha=0.75, ExternalForce="0,0,0") {
    angularDampingOverride=0.98
    bGravityOverrideInSimSpace=true
    bOverrideAngularDamping=true
    bOverrideLinearDamping=true
    gravityScale=0.35
    lODThreshold=1
    linearDampingOverride=0.98
    numSolverIterationsPostUpdate=4
    numSolverIterationsPreUpdate=16
    simulationSpace=RootRelative
    alphaBoolBlend {blendOption=Linear}
    boundBone {boneName=dyn_coat_l}
    componentAppliedLinearAccClamp=<1000 1000 1000>
    componentLinearAccScale=<0.5 0.5 0.5>
    componentLinearVelScale=<0.5 0.5 0.5>
    physicsBodyDefinitions [{
      sphereCollisionRadius=0
      boundBone {boneName=dyn_coat_l}
      boxExtents=<4 4 4>
      constraintSetup {angularLimitsMax=<5 35 15>, angularLimitsMin=<-45 -15 -15>, angularTarget=<0 0 0>}
      localJointOffset=<0 0 0.5>
     }]
   }
   n11 = Anim Dynamics - Bone: dyn_coat_front_l(ComponentPose=@n10, Alpha=0.75, ExternalForce="0,0,0") {
    angularDampingOverride=0.98
    bGravityOverrideInSimSpace=true
    bOverrideAngularDamping=true
    bOverrideLinearDamping=true
    gravityScale=0.35
    lODThreshold=1
    linearDampingOverride=0.98
    numSolverIterationsPostUpdate=4
    numSolverIterationsPreUpdate=16
    simulationSpace=RootRelative
    alphaBoolBlend {blendOption=Linear}
    boundBone {boneName=dyn_coat_front_l}
    physicsBodyDefinitions [{
      sphereCollisionRadius=0
      boundBone {boneName=dyn_coat_front_l}
      boxExtents=<5 5 5>
      constraintSetup {angularLimitsMax=<5 15 15>, angularLimitsMin=<-45 -15 -15>, angularTarget=<0 0 0>}
      localJointOffset=<0 0 0.5>
     }]
   }
   n12 = "Anim Dynamics - Chain: dyn_shorts_in_l -> dyn_shorts_in_02_l"(ComponentPose=@n11, Alpha=@n3, ExternalForce="0,0,0") {
    angularDampingOverride=0.999
    angularSpringConstant=100
    bAngularSpring=true
    bChain=true
    bGravityOverrideInSimSpace=true
    bOverrideAngularDamping=true
    numSolverIterationsPostUpdate=4
    numSolverIterationsPreUpdate=16
    simulationSpace=RootRelative
    alphaBoolBlend {blendOption=Linear}
    boundBone {boneName=dyn_shorts_in_l}
    chainEnd {boneName=dyn_shorts_in_02_l}
    physicsBodyDefinitions [{
      sphereCollisionRadius=0
      boundBone {boneName=dyn_shorts_in_l}
      boxExtents=<25 30 12>
      constraintSetup {angularTargetAxis=AxisZ, twistAxis=AxisZ, angularLimitsMax=<10 10 5>, angularLimitsMin=<-10 -10 -5>, angularTarget=<0 0 1>}
      localJointOffset=<0 0 -3.5>
     }]
   }
   n13 = "Anim Dynamics - Chain: dyn_shorts_out_l -> dyn_shorts_out_02_l"(ComponentPose=@n12, Alpha=@n3, ExternalForce="0,0,0") {
    angularDampingOverride=0.999
    angularSpringConstant=100
    bAngularSpring=true
    bChain=true
    bGravityOverrideInSimSpace=true
    bOverrideAngularDamping=true
    numSolverIterationsPostUpdate=4
    numSolverIterationsPreUpdate=16
    simulationSpace=RootRelative
    alphaBoolBlend {blendOption=Linear}
    boundBone {boneName=dyn_shorts_out_l}
    chainEnd {boneName=dyn_shorts_out_02_l}
    physicsBodyDefinitions [{
      sphereCollisionRadius=0
      boundBone {boneName=dyn_shorts_out_l}
      boxExtents=<50 50 30>
      constraintSetup {angularTargetAxis=AxisZ, twistAxis=AxisZ, angularLimitsMax=<10 10 5>, angularLimitsMin=<-10 -10 -5>, angularTarget=<0 0 1>}
      localJointOffset=<0 0 -6.5>
     }]
   }
   n14 = "Anim Dynamics - Chain: dyn_shorts_fr_l -> dyn_shorts_fr_02_l"(ComponentPose=@n13, Alpha=@n3, ExternalForce="0,0,0") {
    angularDampingOverride=0.999
    angularSpringConstant=300
    bAngularSpring=true
    bChain=true
    bGravityOverrideInSimSpace=true
    bOverrideAngularDamping=true
    numSolverIterationsPostUpdate=4
    numSolverIterationsPreUpdate=16
    simulationSpace=RootRelative
    alphaBoolBlend {blendOption=Linear}
    boundBone {boneName=dyn_shorts_fr_l}
    chainEnd {boneName=dyn_shorts_fr_02_l}
    physicsBodyDefinitions [{
      sphereCollisionRadius=0
      boundBone {boneName=dyn_shorts_fr_l}
      boxExtents=<18 25 8>
      constraintSetup {angularTargetAxis=AxisZ, twistAxis=AxisZ, angularLimitsMax=<10 0 5>, angularLimitsMin=<-10 -20 -5>, angularTarget=<0 0 1>}
      localJointOffset=<0 0 -5>
     }]
   }
   n15 = "Anim Dynamics - Chain: dyn_shorts_bk_l -> dyn_shorts_bk_02_l"(ComponentPose=@n14, Alpha=@n3, ExternalForce="0,0,0") {
    angularDampingOverride=0.999
    angularSpringConstant=300
    bAngularSpring=true
    bChain=true
    bGravityOverrideInSimSpace=true
    bOverrideAngularDamping=true
    numSolverIterationsPostUpdate=4
    numSolverIterationsPreUpdate=16
    simulationSpace=RootRelative
    alphaBoolBlend {blendOption=Linear}
    boundBone {boneName=dyn_shorts_bk_l}
    chainEnd {boneName=dyn_shorts_bk_02_l}
    physicsBodyDefinitions [{
      sphereCollisionRadius=0
      boundBone {boneName=dyn_shorts_bk_l}
      boxExtents=<40 38 15>
      constraintSetup {angularTargetAxis=AxisZ, twistAxis=AxisZ, angularLimitsMax=<10 0 5>, angularLimitsMin=<-10 -20 -5>, angularTarget=<0 0 1>}
      localJointOffset=<0 0 -3.5>
     }]
   }
   n16 = "Anim Dynamics - Chain: dyn_shorts_in_r -> dyn_shorts_in_02_r"(ComponentPose=@n15, Alpha=@n3, ExternalForce="0,0,0") {
    angularDampingOverride=0.999
    angularSpringConstant=100
    bAngularSpring=true
    bChain=true
    bGravityOverrideInSimSpace=true
    bOverrideAngularDamping=true
    numSolverIterationsPostUpdate=4
    numSolverIterationsPreUpdate=16
    simulationSpace=RootRelative
    alphaBoolBlend {blendOption=Linear}
    boundBone {boneName=dyn_shorts_in_r}
    chainEnd {boneName=dyn_shorts_in_02_r}
    physicsBodyDefinitions [{
      sphereCollisionRadius=0
      boundBone {boneName=dyn_shorts_in_r}
      boxExtents=<25 29 12>
      constraintSetup {angularTargetAxis=AxisZ, twistAxis=AxisZ, angularLimitsMax=<10 10 5>, angularLimitsMin=<-10 -10 -5>, angularTarget=<0 0 1>}
      localJointOffset=<0 0 3.5>
     }]
   }
   n17 = "Anim Dynamics - Chain: dyn_shorts_out_r -> dyn_shorts_out_02_r"(ComponentPose=@n16, Alpha=@n3, ExternalForce="0,0,0") {
    angularDampingOverride=0.999
    angularSpringConstant=100
    bAngularSpring=true
    bChain=true
    bGravityOverrideInSimSpace=true
    bOverrideAngularDamping=true
    numSolverIterationsPostUpdate=4
    numSolverIterationsPreUpdate=16
    simulationSpace=RootRelative
    alphaBoolBlend {blendOption=Linear}
    boundBone {boneName=dyn_shorts_out_r}
    chainEnd {boneName=dyn_shorts_out_02_r}
    physicsBodyDefinitions [{
      sphereCollisionRadius=0
      boundBone {boneName=dyn_shorts_out_r}
      boxExtents=<50 50 30>
      constraintSetup {angularTargetAxis=AxisZ, twistAxis=AxisZ, angularLimitsMax=<10 10 5>, angularLimitsMin=<-10 -10 -5>, angularTarget=<0 0 1>}
      localJointOffset=<0 0 6.5>
     }]
   }
   n18 = "Anim Dynamics - Chain: dyn_shorts_fr_r -> dyn_shorts_fr_02_r"(ComponentPose=@n17, Alpha=@n3, ExternalForce="0,0,0") {
    angularDampingOverride=0.999
    angularSpringConstant=300
    bAngularSpring=true
    bChain=true
    bGravityOverrideInSimSpace=true
    bOverrideAngularDamping=true
    numSolverIterationsPostUpdate=4
    numSolverIterationsPreUpdate=16
    simulationSpace=RootRelative
    alphaBoolBlend {blendOption=Linear}
    boundBone {boneName=dyn_shorts_fr_r}
    chainEnd {boneName=dyn_shorts_fr_02_r}
    physicsBodyDefinitions [{
      sphereCollisionRadius=0
      boundBone {boneName=dyn_shorts_fr_r}
      boxExtents=<26 22 8>
      constraintSetup {angularTargetAxis=AxisZ, twistAxis=AxisZ, angularLimitsMax=<10 0 5>, angularLimitsMin=<-10 -20 -5>, angularTarget=<0 0 1>}
      localJointOffset=<0 0 5>
     }]
   }
   n19 = "Anim Dynamics - Chain: dyn_shorts_bk_r -> dyn_shorts_bk_02_r"(ComponentPose=@n18, Alpha=@n3, ExternalForce="0,0,0") {
    angularDampingOverride=0.999
    angularSpringConstant=300
    bAngularSpring=true
    bChain=true
    bGravityOverrideInSimSpace=true
    bOverrideAngularDamping=true
    numSolverIterationsPostUpdate=4
    numSolverIterationsPreUpdate=16
    simulationSpace=RootRelative
    alphaBoolBlend {blendOption=Linear}
    boundBone {boneName=dyn_shorts_bk_r}
    chainEnd {boneName=dyn_shorts_bk_02_r}
    physicsBodyDefinitions [{
      sphereCollisionRadius=0
      boundBone {boneName=dyn_shorts_bk_r}
      boxExtents=<40 36 15>
      constraintSetup {angularTargetAxis=AxisZ, twistAxis=AxisZ, angularLimitsMax=<0 0 5>, angularLimitsMin=<0 -20 -5>, angularTarget=<0 0 1>}
      localJointOffset=<0 0 3.5>
     }]
   }
   n20 = "Transform (Modify) Bone - Bone: driver_collar_frt_01_r"(ComponentPose=@n19, Alpha=Get ArmUp_R() -> (ArmUp_R:float), Translation="0,0,0", Rotation="-40,-4,20", Scale="0.500,0.500,0.500") {
    rotationMode=BMM_Additive
    rotationSpace=BCS_BoneSpace
    scaleMode=BMM_Additive
    scaleSpace=BCS_BoneSpace
    alphaBoolBlend {blendOption=Linear}
    alphaScaleBiasClamp {bInterpResult=true, interpSpeedDecreasing=30, interpSpeedIncreasing=30}
    boneToModify {boneName=driver_collar_frt_01_r}
    rotation=<rot -40 -4 20>
    scale=<0.5 0.5 0.5>
   }
   n21 = "Transform (Modify) Bone - Bone: driver_collar_frt_02_r"(ComponentPose=@n20, Alpha=@n2, Translation="-10,-0.000,-0.000", Rotation="0,0,0", Scale="0.25,0.25,0.25") {
    rotationMode=BMM_Additive
    rotationSpace=BCS_BoneSpace
    scaleMode=BMM_Additive
    scaleSpace=BCS_BoneSpace
    translationMode=BMM_Additive
    translationSpace=BCS_BoneSpace
    alphaBoolBlend {blendOption=Linear}
    alphaScaleBiasClamp {bInterpResult=true, interpSpeedDecreasing=30, interpSpeedIncreasing=30}
    boneToModify {boneName=driver_collar_frt_02_r}
    scale=<0.25 0.25 0.25>
    translation {x=-10}
   }
   n22 = "Transform (Modify) Bone - Bone: parent_collar_bck_01"(
    ComponentPose=@n21
    Alpha=Get HeadRotBack() -> (HeadRotBack:float)
    Translation="-5,0.000,-4"
    Rotation="-40,0.000,0.000"
    Scale="0.9,0.9,0.9"
   ) {
    rotationMode=BMM_Additive
    rotationSpace=BCS_BoneSpace
    scaleMode=BMM_Additive
    scaleSpace=BCS_BoneSpace
    translationMode=BMM_Additive
    translationSpace=BCS_BoneSpace
    alphaBoolBlend {blendOption=Linear}
    alphaScaleBiasClamp {bInterpResult=true, interpSpeedDecreasing=30, interpSpeedIncreasing=30}
    boneToModify {boneName=parent_collar_bck_01}
    rotation {pitch=-40}
    scale=<0.9 0.9 0.9>
    translation {x=-5, z=-4}
   }
   n23 = "Transform (Modify) Bone - Bone: collar_mid_01_l"(ComponentPose=@n22, Alpha=@n2, Translation="0,0,5", Rotation="5,30,-1.590", Scale="1,1,1") {
    rotationMode=BMM_Additive
    rotationSpace=BCS_BoneSpace
    scaleMode=BMM_Additive
    scaleSpace=BCS_BoneSpace
    translationMode=BMM_Additive
    alphaBoolBlend {blendOption=Linear}
    alphaScaleBiasClamp {bInterpResult=true, interpSpeedDecreasing=30, interpSpeedIncreasing=30}
    boneToModify {boneName=collar_mid_01_l}
    rotation=<rot 5 30 -1.59>
    translation {z=5}
   }
   n24 = "Transform (Modify) Bone - Bone: collar_mid_01_r"(ComponentPose=@n23, Alpha=@n2, Translation="0,0,1", Rotation="5,5,-1.590", Scale="1,1,1") {
    rotationMode=BMM_Additive
    rotationSpace=BCS_BoneSpace
    scaleMode=BMM_Additive
    scaleSpace=BCS_BoneSpace
    translationMode=BMM_Additive
    alphaBoolBlend {blendOption=Linear}
    alphaScaleBiasClamp {bInterpResult=true, interpSpeedDecreasing=30, interpSpeedIncreasing=30}
    boneToModify {boneName=collar_mid_01_r}
    rotation=<rot 5 5 -1.59>
    translation {z=1}
   }
   n25 = "Transform (Modify) Bone - Bone: collar_frt_01_l"(ComponentPose=@n24, Alpha=@n2, Translation="0,0,0", Rotation="-10,4.150,13", Scale="0.9,0.9,0.9") {
    rotationMode=BMM_Additive
    rotationSpace=BCS_BoneSpace
    scaleMode=BMM_Additive
    scaleSpace=BCS_BoneSpace
    alphaBoolBlend {blendOption=Linear}
    alphaScaleBiasClamp {bInterpResult=true, interpSpeedDecreasing=30, interpSpeedIncreasing=30}
    boneToModify {boneName=collar_frt_01_l}
    rotation=<rot -10 4.15 13>
    scale=<0.9 0.9 0.9>
   }
   n26 = Anim Dynamics - Bone: collar_bck_01(ComponentPose=@n25, Alpha=0.3, ExternalForce="0,0,0") {
    angularDampingOverride=0.9995
    angularSpringConstant=200
    bAngularSpring=true
    bGravityOverrideInSimSpace=true
    bOverrideAngularDamping=true
    gravityScale=0.33
    numSolverIterationsPostUpdate=2
    numSolverIterationsPreUpdate=8
    simulationSpace=RootRelative
    alphaBoolBlend {blendOption=Linear}
    boundBone {boneName=collar_bck_01}
    componentAppliedLinearAccClamp=<1000 1000 1000>
    physicsBodyDefinitions [{
      sphereCollisionRadius=0
      boundBone {boneName=collar_bck_01}
      boxExtents=<25 25 10>
      constraintSetup {angularLimitsMax=<15 5 10>, angularLimitsMin=<-15 -5 -10>, angularTarget=<0 0 0>}
      localJointOffset=<0 0 -5>
     }]
   }
   n27 = Get CalfRotBack_L() -> (CalfRotBack_L:float)
   n28 = "Transform (Modify) Bone - Bone: dyn_shorts_main_l"(ComponentPose=@n26, Alpha=@n27, Translation="0.000,-0.000,-4", Rotation="0,0,10", Scale="1,1,1") {
    rotationMode=BMM_Additive
    rotationSpace=BCS_BoneSpace
    scaleMode=BMM_Additive
    scaleSpace=BCS_BoneSpace
    translationMode=BMM_Additive
    translationSpace=BCS_BoneSpace
    alphaBoolBlend {blendOption=Linear}
    alphaScaleBiasClamp {bInterpResult=true, interpSpeedDecreasing=25, interpSpeedIncreasing=25}
    boneToModify {boneName=dyn_shorts_main_l}
    rotation {roll=10}
    translation {z=-4}
   }
   n29 = "Transform (Modify) Bone - Bone: dyn_shorts_bk_l"(ComponentPose=@n28, Alpha=@n27, Translation="-0.247,0.000,20", Rotation="24.730,-3.101,-11.925", Scale="0.9,0.9,0.9") {
    rotationMode=BMM_Additive
    rotationSpace=BCS_BoneSpace
    scaleMode=BMM_Additive
    scaleSpace=BCS_BoneSpace
    translationMode=BMM_Additive
    translationSpace=BCS_BoneSpace
    alphaBoolBlend {blendOption=Linear}
    alphaScaleBiasClamp {bInterpResult=true, interpSpeedDecreasing=25, interpSpeedIncreasing=25}
    boneToModify {boneName=dyn_shorts_bk_l}
    rotation=<rot 24.73 -3.101 -11.925>
    scale=<0.9 0.9 0.9>
    translation {x=-0.247, z=20}
   }
   n30 = "Transform (Modify) Bone - Bone: dyn_shorts_out_l"(ComponentPose=@n29, Alpha=@n27, Translation="-0.000,0.022,7", Rotation="8,0.000,0.000", Scale="1,1,1") {
    rotationMode=BMM_Additive
    rotationSpace=BCS_BoneSpace
    scaleMode=BMM_Additive
    scaleSpace=BCS_BoneSpace
    translationMode=BMM_Additive
    translationSpace=BCS_BoneSpace
    alphaBoolBlend {blendOption=Linear}
    alphaScaleBiasClamp {bInterpResult=true, interpSpeedDecreasing=25, interpSpeedIncreasing=25}
    boneToModify {boneName=dyn_shorts_out_l}
    rotation {pitch=8}
    translation {y=0.022, z=7}
   }
   n31 = "Transform (Modify) Bone - Bone: dyn_shorts_in_l"(ComponentPose=@n30, Alpha=@n27, Translation="0.000,-0.033,11", Rotation="-12,0.000,0.000", Scale="1,1,1") {
    rotationMode=BMM_Additive
    rotationSpace=BCS_BoneSpace
    scaleMode=BMM_Additive
    scaleSpace=BCS_BoneSpace
    translationMode=BMM_Additive
    translationSpace=BCS_BoneSpace
    alphaBoolBlend {blendOption=Linear}
    alphaScaleBiasClamp {bInterpResult=true, interpSpeedDecreasing=25, interpSpeedIncreasing=25}
    boneToModify {boneName=dyn_shorts_in_l}
    rotation {pitch=-12}
    translation {y=-0.033, z=11}
   }
   n32 = "Transform (Modify) Bone - Bone: dyn_shorts_fr_l"(ComponentPose=@n31, Alpha=@n27, Translation="-9,0.400,0", Rotation="-13,0,0", Scale="1,1,1") {
    rotationMode=BMM_Additive
    rotationSpace=BCS_BoneSpace
    scaleMode=BMM_Additive
    scaleSpace=BCS_BoneSpace
    translationMode=BMM_Additive
    translationSpace=BCS_BoneSpace
    alphaBoolBlend {blendOption=Linear}
    alphaScaleBiasClamp {bInterpResult=true, interpSpeedDecreasing=25, interpSpeedIncreasing=25}
    boneToModify {boneName=dyn_shorts_fr_l}
    rotation {pitch=-13}
    translation=<-9 0.4>
   }
   n33 = "Transform (Modify) Bone - Bone: dyn_shorts_main_r"(ComponentPose=@n32, Alpha=@n1, Translation="0.000,-0.000,3", Rotation="0,0,10", Scale="1,1,1") {
    rotationMode=BMM_Additive
    rotationSpace=BCS_BoneSpace
    scaleMode=BMM_Additive
    scaleSpace=BCS_BoneSpace
    translationMode=BMM_Additive
    translationSpace=BCS_BoneSpace
    alphaBoolBlend {blendOption=Linear}
    alphaScaleBiasClamp {bInterpResult=true, interpSpeedDecreasing=25, interpSpeedIncreasing=25}
    boneToModify {boneName=dyn_shorts_main_r}
    rotation {roll=10}
    translation {z=3}
   }
   n34 = "Transform (Modify) Bone - Bone: dyn_shorts_bk_r"(ComponentPose=@n33, Alpha=@n1, Translation="-0.247,0.000,-20", Rotation="24.730,-3.101,-11.925", Scale="0.9,0.9,0.9") {
    rotationMode=BMM_Additive
    rotationSpace=BCS_BoneSpace
    scaleMode=BMM_Additive
    scaleSpace=BCS_BoneSpace
    translationMode=BMM_Additive
    translationSpace=BCS_BoneSpace
    alphaBoolBlend {blendOption=Linear}
    alphaScaleBiasClamp {bInterpResult=true, interpSpeedDecreasing=25, interpSpeedIncreasing=25}
    boneToModify {boneName=dyn_shorts_bk_r}
    rotation=<rot 24.73 -3.101 -11.925>
    scale=<0.9 0.9 0.9>
    translation {x=-0.247, z=-20}
   }
   n35 = "Transform (Modify) Bone - Bone: dyn_shorts_out_r"(ComponentPose=@n34, Alpha=@n1, Translation="-0.000,0.022,-7", Rotation="8,0.000,0.000", Scale="1,1,1") {
    rotationMode=BMM_Additive
    rotationSpace=BCS_BoneSpace
    scaleMode=BMM_Additive
    scaleSpace=BCS_BoneSpace
    translationMode=BMM_Additive
    translationSpace=BCS_BoneSpace
    alphaBoolBlend {blendOption=Linear}
    alphaScaleBiasClamp {bInterpResult=true, interpSpeedDecreasing=25, interpSpeedIncreasing=25}
    boneToModify {boneName=dyn_shorts_out_r}
    rotation {pitch=8}
    translation {y=0.022, z=-7}
   }
   n36 = "Transform (Modify) Bone - Bone: dyn_shorts_in_r"(ComponentPose=@n35, Alpha=@n1, Translation="0.000,-0.033,-11", Rotation="0,0,0", Scale="1,1,1") {
    rotationMode=BMM_Additive
    rotationSpace=BCS_BoneSpace
    scaleMode=BMM_Additive
    scaleSpace=BCS_BoneSpace
    translationMode=BMM_Additive
    translationSpace=BCS_BoneSpace
    alphaBoolBlend {blendOption=Linear}
    alphaScaleBiasClamp {bInterpResult=true, interpSpeedDecreasing=25, interpSpeedIncreasing=25}
    boneToModify {boneName=dyn_shorts_in_r}
    translation {y=-0.033, z=-11}
   }
   n37 = "Transform (Modify) Bone - Bone: dyn_shorts_fr_r"(ComponentPose=@n36, Alpha=@n1, Translation="7,0.400,0.109", Rotation="-8,0,0", Scale="1,1,1") {
    rotationMode=BMM_Additive
    rotationSpace=BCS_BoneSpace
    scaleMode=BMM_Additive
    scaleSpace=BCS_BoneSpace
    translationMode=BMM_Additive
    translationSpace=BCS_BoneSpace
    alphaBoolBlend {blendOption=Linear}
    alphaScaleBiasClamp {bInterpResult=true, interpSpeedDecreasing=25, interpSpeedIncreasing=25}
    boneToModify {boneName=dyn_shorts_fr_r}
    rotation {pitch=-8}
    translation=<7 0.4 0.109>
   }
   n38 = "Transform (Modify) Bone - Bone: knob"(ComponentPose=@n37, Alpha=1, Translation="-1.75,0.000,0.000", Rotation="0,0,0", Scale="0.9,0.9,0.9") {
    scaleMode=BMM_Additive
    translationMode=BMM_Additive
    alphaBoolBlend {blendOption=Linear}
    boneToModify {boneName=knob}
    scale=<0.9 0.9 0.9>
    translation {x=-1.75}
   }
   n39 = "Constraint - parent_strand_01_r (PPP)"!(ComponentPose=@n38, Alpha=1, ConstraintWeights_0=0.0266, ConstraintWeights_1:float=0.676, ConstraintWeights_2:float=0.2974) -> (Pose) {
    alphaBoolBlend {blendOption=Linear}
    boneToModify {boneName=parent_strand_01_r}
    constraintSetup [{targetBone {boneName=spine_01}}, {targetBone {boneName=thigh_twist_01_r}}, {targetBone {boneName=pelvis}}]
    constraintWeights [1,1,1]
   }
   n40 = "Transform (Modify) Bone - Bone: dyn_thigh_twist_01_r"(ComponentPose=@n39, Alpha=1, Translation="1.25,0,0", Rotation="0,0,0", Scale="1,1,1") {
    translationMode=BMM_Additive
    translationSpace=BCS_ParentBoneSpace
    alphaBoolBlend {blendOption=Linear}
    boneToModify {boneName=dyn_thigh_twist_01_r}
    translation {x=1.25}
   }
   n41 = RigidBody(ComponentPose=@n40) {
    bClampLinearTranslationLimitToRefPose=true
    bOverrideWorldGravity=true
    overridePhysicsAsset=&/M_MED_Dragon_Mask_Physics
    simulationSpace=BaseBoneSpace
    alphaBoolBlend {blendOption=Linear}
    baseBoneRef {boneName=pelvis}
    componentAppliedLinearAccClamp=<5 5 5>
    componentLinearAccScale=<1 1 1>
    componentLinearVelScale=<1 1 1>
    overrideWorldGravity {z=-980}
   }
   n42 = "Transform (Modify) Bone - Bone: dyn_thigh_twist_01_l"(ComponentPose=@n41, Alpha=1, Translation="-1.75,0,0", Rotation="0,0,0", Scale="1,1,1") {
    translationMode=BMM_Additive
    translationSpace=BCS_ParentBoneSpace
    alphaBoolBlend {blendOption=Linear}
    boneToModify {boneName=dyn_thigh_twist_01_l}
    translation {x=-1.75}
   }
   n43 = RigidBody(ComponentPose=@n42) {
    bOverrideWorldGravity=true
    overridePhysicsAsset=&/M_MED_Dragon_Mask_TasselsLPhysics
    simulationSpace=BaseBoneSpace
    alphaBoolBlend {blendOption=Linear}
    baseBoneRef {boneName=thigh_l}
    componentAppliedLinearAccClamp=<100 100 100>
    componentLinearAccScale=<1 1 1>
    componentLinearVelScale=<1 1 1>
    overrideWorldGravity {z=-980}
   }
   Output Pose(Result=Component To Local(ComponentPose=@n43)) {layerGroup=None}
  ]}
 EventGraph {links [], nodes []}
]
variables []
