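# Physics asset for the face accessory of the male medium Dragon Mask character cosmetic, giving its hair pieces physical bodies for secondary motion.
# Five capsule bodies cover the left, right and middle hair bones, alongside a small capsule on the head.
# Each hair body is constrained directly to the head rather than chained, with wide swing limits and tighter twist limits.
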PhysicsAsset /BRCosmetics/Characters/Player/Male/Medium/Bodies/M_MED_DragonMask/Meshes/Parts/M_MED_Dragon_Mask_FaceAcc_Physics
bodies {
 hairA_r [{shapeType=Capsule, length=9, radius=3.5, center=<0 -1 0>, rotation=<rot 74.9998 89.9999 0>}]
 hair_l [{shapeType=Capsule, length=8.25, radius=4.25, center=<0 -1 0>, rotation=<rot 74.9998 89.9999 0>}]
 hair_mid [{shapeType=Capsule, length=6.25, radius=4.25, center=<0 -1.25 0>, rotation=<rot 79.166 89.9853 179.9999>}]
 hair_midA [{shapeType=Capsule, length=5.5, radius=4.25, center=<0 -1 0.25>, rotation=<rot 74.9997 89.9997 0>}]
 hair_r [{shapeType=Capsule, length=6.25, radius=4.25, center=<0 -0.9853 0>, rotation=<rot 59.999 89.9992 -180>}]
 head [{shapeType=Capsule, length=1, radius=2}]
}
constraints [
 {bone1Name=hair_mid, bone2Name=head, swing1LimitDegrees=60, swing2LimitDegrees=60, twistLimitDegrees=15}
 {bone1Name=hair_l, bone2Name=head, swing1LimitDegrees=50, swing2LimitDegrees=60, twistLimitDegrees=10}
 {bone1Name=hair_r, bone2Name=head, swing1LimitDegrees=60, swing2LimitDegrees=60, twistLimitDegrees=15}
 {bone1Name=hair_midA, bone2Name=head, swing1LimitDegrees=50, swing2LimitDegrees=60, twistLimitDegrees=20}
 {bone1Name=hairA_r, bone2Name=head, swing1LimitDegrees=50, swing2LimitDegrees=60, twistLimitDegrees=10}
]
